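; Blueprint for the Asteria NeonCity Lanterns A prop with time-of-day lighting support.
; Its construction script checks a toggle for using the TOD component and, when enabled, runs the TODLightingComponent's construct step with a test day phase value.
; BeginPlay, overlap and Tick events are empty, so this graph holds no runtime logic of its own.
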(event EventBeginPlay)

(event Collision|EventActorBeginOverlap
  (params (OtherActor :type Actor%20Object%20Reference)))

(event EventTick
  (params (DeltaSeconds :type Float%20%28single-precision%29)))

(fn ConstructionScript
  (if (Variables|Default|GetUse%20of%20TOD%20component)
    (Class|TODLightingComponent|RunConstruct (Variables|Default|GetTODLightingComponent) (Variables|Default|GetTest%20Day%20Phase_0))))
