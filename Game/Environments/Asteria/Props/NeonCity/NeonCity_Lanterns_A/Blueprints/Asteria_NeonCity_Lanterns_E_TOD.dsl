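; Neon City lantern prop Blueprint for the Asteria environment, with optional time-of-day (TOD) lighting.
; The construction script runs the parent's, then, if the TOD component flag is on, calls RunConstruct on the TODLightingComponent with a test day phase.
; BeginPlay, Tick and overlap events are disabled stubs, so the only logic here runs at construction.
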(event EventBeginPlay
  (|CallParentFunction) ; This node is disabled and will not be called.\nDrag off pins to build functionality.
)

(event Collision|EventActorBeginOverlap
  (params (OtherActor :type Actor%20Object%20Reference))
  (|CallParentFunction OtherActor) ; This node is disabled and will not be called.\nDrag off pins to build functionality.
)

(event EventTick
  (params (DeltaSeconds :type Float%20%28single-precision%29))
  (|CallParentFunction DeltaSeconds) ; This node is disabled and will not be called.\nDrag off pins to build functionality.
)

(fn ConstructionScript
  (|CallParentFunction)
  (if (Variables|Default|GetUse%20of%20TOD%20Component)
    (Class|TODLightingComponent|RunConstruct (Variables|Default|GetTODLightingComponent) (Variables|Default|GetTest%20Day%20Phase_0))))
